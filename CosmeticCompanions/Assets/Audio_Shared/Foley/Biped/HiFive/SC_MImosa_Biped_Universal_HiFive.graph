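# Sound cue for the high-five foley of biped Mimosa companion pets.
# A Random node picks one of three equally weighted wave variations, and a Modulator applies slight pitch variation at constant volume.
# It routes to the creature pets sound class, uses companion foley attenuation, and shares the biped footstep concurrency group.
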SoundCue /CosmeticCompanions/Assets/Audio_Shared/Foley/Biped/HiFive/SC_MImosa_Biped_Universal_HiFive
root=Modulator_0
nodes [
 {id=Modulator_0, children [Random_0], properties {pitchMax=1.1, pitchMin=0.9, volumeMax=1, volumeMin=1}}
 {id=Random_0, children [WavePlayer_0,WavePlayer_1,WavePlayer_2], properties {weights [1,1,1]}}
 {id=WavePlayer_0, properties {soundWaveAssetPtr=&/Mimosa_Biped_Universal_HiFive_01}}
 {id=WavePlayer_1, properties {soundWaveAssetPtr=&/Mimosa_Biped_Universal_HiFive_02}}
 {id=WavePlayer_2, properties {soundWaveAssetPtr=&/Mimosa_Biped_Universal_HiFive_03}}
]
properties {
 attenuationSettings=/Game/Sounds/Attenuation/AS_Companion_Foley
 maxDistance=1200
 priority=0
 soundClassObject=&Fort_Creatures_Pets
 volumeMultiplier=1
 concurrencySet [/CosmeticCompanions/Assets/Audio_Shared/Foley/Biped/CONC_Mimosa_Biped_Footstep]
}
