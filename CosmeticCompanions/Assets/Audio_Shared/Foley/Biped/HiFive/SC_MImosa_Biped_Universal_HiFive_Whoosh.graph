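# Sound cue for the high-five whoosh foley of biped cosmetic companions.
# A Random node picks one of three whoosh wave variations with equal weight.
# It plays at reduced volume and zero priority through the pets sound class and companion foley attenuation.
# Its voice limiting uses the biped footstep concurrency asset rather than one specific to high-fives.
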SoundCue /CosmeticCompanions/Assets/Audio_Shared/Foley/Biped/HiFive/SC_MImosa_Biped_Universal_HiFive_Whoosh
root=Random_0
nodes [
 {id=Random_0, children [WavePlayer_3,WavePlayer_4,WavePlayer_5], properties {weights [1,1,1]}}
 {
  id=WavePlayer_3
  properties {soundWaveAssetPtr=&/Mimosa_Biped_Universal_HiFive_Whoosh_01}
 }
 {
  id=WavePlayer_4
  properties {soundWaveAssetPtr=&/Mimosa_Biped_Universal_HiFive_Whoosh_02}
 }
 {
  id=WavePlayer_5
  properties {soundWaveAssetPtr=&/Mimosa_Biped_Universal_HiFive_Whoosh_03}
 }
]
properties {
 attenuationSettings=/Game/Sounds/Attenuation/AS_Companion_Foley
 maxDistance=1200
 priority=0
 soundClassObject=&Fort_Creatures_Pets
 volumeMultiplier=0.6
 concurrencySet [/CosmeticCompanions/Assets/Audio_Shared/Foley/Biped/CONC_Mimosa_Biped_Footstep]
}
